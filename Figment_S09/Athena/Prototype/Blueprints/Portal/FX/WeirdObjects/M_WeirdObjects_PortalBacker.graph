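# Lit material for the prototype WeirdObjects portal backer, a glowing, rippling surface that intensifies with the Active parameter.
# Two drifting sphere masks plus a parallax center mask form an HDR glow, cut by two sine-warped, panning electric-noise layers that also drive base color and a 0.3-to-0 scalar blend.
# Panning tile normals are angle-blended for surface detail.
# Sphere-masked, distorted noise displaces vertices along local Y, added to Fortnite hit-and-shrink offset.
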Material /Figment_S09/Athena/Prototype/Blueprints/Portal/FX/WeirdObjects/M_WeirdObjects_PortalBacker
graph {
 properties {ditherOpacityMask=true, pixelDepthOffsetMode=PDOM_Legacy, usedShadingModels=Default Lit}
 nodes [
  n1 = Multiply(
   A=Transform(None=Constant3Vector() {constant=<rgba 0 1 0 0>}) {transformSourceType=TRANSFORMSOURCE_Local}
   B=LinearInterpolate(Alpha=ScalarParameter() {parameterName=Active}) {constA=10, constB=100}
  )
  n2 = TextureCoordinate()
  n3 = Panner(Coordinate=@n2) {speedY=-0.06}
  n4 = Multiply(A=ComponentMask(None=TextureSample(UVs=@n3) {texture=/Game/VisualThreatManager/StormVisuals/Textures/2ndInfectedMapLinear}.R) {g=true, r=true}) {constB=0.02}
  n5 = TextureSample!(
   UVs=Panner(Coordinate=Add(A=@n4, B=@n2)) {speedY=-0.1}
   Tex
   Level=ScalarParameter() {defaultValue=2, parameterName=MipLevel}
   Apply View MipBias
  ) {
   mipValueMode=TMVM_MipLevel
   samplerType=SAMPLERTYPE_LinearColor
   texture=/Game/Effects/Fort_Effects/Gameplay/SmokeDustExplosions/Textures/T_NoisePack_LinearColor
  }
  n6 = TextureCoordinate()
  n7 = Constant() {r=0.5}
  n8 = ScalarParameter() {defaultValue=0.4, parameterName=Radius}
  n9 = ScalarParameter() {defaultValue=0.1, parameterName=Hardness}
  Add(
   A=Multiply(A=Multiply(A=@n1, B=@n5.G), B=SphereMask(A=@n6, B=@n7, Radius=@n8, Hardness=@n9) {hardnessPercent=0})
   B=MaterialFunctionCall!(
    WSDeformationCenter=VectorParameter() {bUseCustomPrimitiveData=true, group=z_code, parameterName=WSDeformationCenter, primitiveDataIndex=20}.RGB
    Code_DestructionAmount=ScalarParameter() {group=z_code, parameterName=Code_DestructionAmount}
    Code_UseUvForDamage
    Code_UseHitEffects=StaticBoolParameter() {defaultValue=true, group=z_code, parameterName=Code_UseHitEffects}
    Code_Emissive
    WSDamageRadius=ScalarParameter() {bUseCustomPrimitiveData=true, defaultValue=1, group=z_code, parameterName=WSDamageRadius, primitiveDataIndex=26}
    CrackDamageShader
    Diffuse
    ScalarModDamageCracks
    ScalarOffsetforDamageCracks
    ExposeLerpFunctionForInvertingTheHitDistanceCalc
    InvertDamageFalloff
    CrackWorldPosition
    WorldPosition
    DistanceCullFadeAnimationOffset=ScalarParameter() {bUseCustomPrimitiveData=true, group=z_code, parameterName=DistanceCullFadeAnimationOffset, primitiveDataIndex=12}
   ) {materialFunction=/Game/Packages/jl_tests/Fortnite_HitEffects}.Hit And Shrink WPO
  )
  n10 = TextureCoordinate() {vTiling=2}
  n11 = Multiply(A=Time(), B=ScalarParameter() {defaultValue=0.125, parameterName=Speed})
  n12 = LinearInterpolate(Alpha=ConstantBiasScale(None=Sine(None=@n11) {period=2}))
  n13 = Multiply(A=Time(), B=ScalarParameter() {defaultValue=0.1368, parameterName=Speed2})
  n14 = LinearInterpolate(Alpha=ConstantBiasScale(None=Cosine(None=@n13) {period=1.5}))
  n15 = LinearInterpolate(Alpha=ConstantBiasScale(None=Cosine(None=@n11) {period=2}))
  n16 = LinearInterpolate(Alpha=ConstantBiasScale(None=Sine(None=@n13) {period=1.5}))
  n17 = Add(
   A=SphereMask(A=@n10, B=AppendVector(A=@n12, B=@n14), Radius=ScalarParameter() {defaultValue=0.4, parameterName=Radius One}) {hardnessPercent=0}
   B=SphereMask(A=@n10, B=AppendVector(A=@n15, B=@n16), Radius=ScalarParameter() {defaultValue=0.35, parameterName=Radius Two}) {hardnessPercent=0}
  )
  n18 = TextureCoordinate()
  n19 = ScalarParameter() {defaultValue=-1, parameterName=Height}
  n20 = Constant() {r=0.5}
  n21 = Add(
   A=Saturate(None=Multiply(A=@n17) {constB=0.5})
   B=SphereMask(A=BumpOffset(Coordinate=@n18, Height=@n19), B=@n20, Radius=@n20) {hardnessPercent=0}
  )
  n22 = VectorParameter() {parameterName=EmissiveColor, defaultValue=<rgba 1.2 0.2 0 0>}
  n23 = Multiply(A=@n21, B=@n22.RGB)
  n24 = LinearInterpolate(Alpha=ScalarParameter() {parameterName=Active}) {constA=1, constB=4}
  n25 = Power(Base=@n21)
  n26 = TextureCoordinate()
  n27 = VectorParameter() {parameterName=UVControl, defaultValue=<rgba 0.1 5 0 -0.25>}
  n28 = AppendVector(A=@n27.RGB, B=@n27.RGB)
  n29 = ComponentMask(None=@n28) {g=true, r=true}
  n30 = ComponentMask(None=@n28) {a=true, b=true}
  n31 = ComponentMask(None=TextureCoordinate()) {r=true}
  n32 = ScalarParameter() {defaultValue=2, parameterName=SineContrast}
  n33 = ConstantBiasScale(None=Sine(None=Time()) {period=13})
  n34 = Saturate(None=Power(Base=Sine(None=@n31) {period=2}, Exp=LinearInterpolate(B=@n32, Alpha=@n33) {constA=1}))
  n35 = Panner(Coordinate=Add(A=Multiply(A=@n26, B=@n29), B=LinearInterpolate(B=@n30, Alpha=@n34))) {bFractionalPart=true, speedX=0.0125}
  n36 = VectorParameter() {parameterName=UVControl_Layer2, defaultValue=<rgba 0.1 2 0 -0.25>}
  n37 = AppendVector(A=@n36.RGB, B=@n36.RGB)
  n38 = Multiply(A=TextureCoordinate(), B=ComponentMask(None=@n37) {g=true, r=true})
  n39 = Sine(None=ComponentMask(None=TextureCoordinate()) {r=true}) {period=2}
  n40 = Time()
  n41 = LinearInterpolate(B=ScalarParameter() {defaultValue=2, parameterName=SineContrast}, Alpha=ConstantBiasScale(None=Sine(None=@n40) {period=10})) {constA=1}
  n42 = LinearInterpolate(B=ComponentMask(None=@n37) {a=true, b=true}, Alpha=Saturate(None=Power(Base=@n39, Exp=@n41)))
  n43 = Panner(Coordinate=BumpOffset(Coordinate=Add(A=@n38, B=@n42), Height=ScalarParameter() {defaultValue=-1, parameterName=Height})) {bFractionalPart=true, speedY=-0.005}
  n44 = Add(
   A=TextureSample!(UVs=@n35, Tex, Apply View MipBias) {samplerType=SAMPLERTYPE_LinearColor, texture=/Game/Characters/Player/Male/Medium/Bodies/M_MED_SRDrift/Textures/T_ElectricNoise}.R
   B=TextureSample(UVs=@n43) {samplerType=SAMPLERTYPE_LinearColor, texture=&T_ElectricNoise}.R
  )
  LinearInterpolate(
   A=MaterialFunctionCall(LDR Input=Saturate(None=@n23), HDR Tint=@n22.RGB, HDR Intensity Multiplier=LinearInterpolate(B=@n24, Alpha=@n25) {constA=1, constB=2}) {materialFunction=&DeriveHDRfromLDR}.RemappedRange
   Alpha=@n44
  ) {constB=0}
  LinearInterpolate(Alpha=@n44) {constA=0.3, constB=0}
  n45 = TextureSample(UVs=@n35) {samplerType=SAMPLERTYPE_Normal, texture=/Game/Effects/Fort_Effects/Textures/Husk/Poison/T_PoisonTile_N}
  n46 = TextureSample(UVs=@n43) {samplerType=SAMPLERTYPE_Normal, texture=&T_PoisonTile_N}
  MaterialFunctionCall(
   BaseNormal=TextureSample(UVs=Panner() {speedY=-0.0025}) {samplerType=SAMPLERTYPE_Normal, texture=&T_PoisonTile_N}.RGB
   AdditionalNormal=Normalize(VectorInput=Add(A=@n45.RGB, B=@n46.RGB))
  ) {materialFunction=&BlendAngleCorrectedNormals}
  Multiply(A=VectorParameter() {parameterName=Color, defaultValue=<rgba 0.01 0.0062 0.0003 0>}.RGB, B=@n44)
  TextureCoordinate()
 ]
}
