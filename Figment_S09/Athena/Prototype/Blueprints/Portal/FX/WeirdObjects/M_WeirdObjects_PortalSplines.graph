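# Translucent unlit, two-sided material for thin spline strands among the portal's weird-object effects.
# SplineThicken widens the strands; vertex colors and UV sines vary thickness and drive static and time-animated offsets, combined with camera offset and Fortnite_HitEffects shrink WPO.
# A panning noise texture distorts UVs for a scrolling streak pattern that sets depth-faded opacity.
# Emissive fakes glints by raising a fake light vector dotted with a flattened water-normal reflection to a high power.
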Material /Figment_S09/Athena/Prototype/Blueprints/Portal/FX/WeirdObjects/M_WeirdObjects_PortalSplines
graph {
 properties {
  bFullyRough=true
  blendMode=BLEND_Translucent
  ditherOpacityMask=true
  pixelDepthOffsetMode=PDOM_Legacy
  shadingModel=MSM_Unlit
  twoSided=true
  usedShadingModels=Unlit
 }
 nodes [
  n1 = TextureCoordinate()
  n2 = ScalarParameter() {defaultValue=0.15, parameterName=BulgeOffset}
  n3 = VertexColor()
  n4 = Add(A=Add(A=ComponentMask(None=@n1) {r=true}, B=LinearInterpolate(B=@n2, Alpha=@n3.G)), B=VertexColor().R)
  n5 = TextureCoordinate()
  n6 = Multiply(A=ComponentMask(None=@n5) {r=true}, B=ScalarParameter() {defaultValue=1.5, parameterName=UTile})
  n7 = VertexColor()
  n8 = MaterialFunctionCall(
   WidthBase=ScalarParameter() {defaultValue=10, parameterName=Width_Base}
   WidthTip=ScalarParameter() {defaultValue=20, parameterName=Width_Top}
   UVs for Thickness=AppendVector(A=Sine(None=@n4), B=ComponentMask(None=@n1) {g=true})
   UVs for texturing=AppendVector(A=Add(A=@n6, B=@n7.R), B=ComponentMask(None=@n5) {g=true})
   Expand U or V UV Channel=StaticBool()
  ) {materialFunction=&SplineThicken}
  n9 = ScalarParameter() {defaultValue=20, parameterName=Camera_Offset}
  n10 = WorldPosition()
  n11 = ComponentMask(None=TextureCoordinate()) {r=true}
  n12 = Add(
   A=@n8.WorldPositionOffset
   B=LinearInterpolate(
    B=MaterialFunctionCall(Offset Amount=@n9, World Position=@n10.XYZ) {materialFunction=/Engine/Functions/Engine_MaterialFunctions02/WorldPositionOffset/CameraOffset}.Camera Offset
    Alpha=Sine(None=@n11) {period=2}
   )
  )
  n13 = Add(
   A=ComponentMask(None=TextureCoordinate()) {r=true}
   B=LinearInterpolate(B=ScalarParameter() {defaultValue=0.15, parameterName=BulgeOffset}, Alpha=VertexColor().G)
  )
  n14 = VertexColor()
  n15 = LinearInterpolate(
   B=Transform(None=VectorParameter() {parameterName=Offset_WPO, defaultValue=<rgba 0 5 0 0>}.RGB) {transformSourceType=TRANSFORMSOURCE_Local}
   Alpha=Sine(None=Add(A=@n13, B=@n14.R))
  )
  n16 = Transform(None=VectorParameter() {parameterName=Offset_WPO, defaultValue=<rgba 0 5 0 0>}.RGB) {transformSourceType=TRANSFORMSOURCE_Local}
  n17 = ScalarParameter() {defaultValue=0.05, parameterName=Outward}
  n18 = VertexColor()
  n19 = Add(A=ComponentMask(None=TextureCoordinate()) {r=true}, B=Multiply(A=Time(), B=Add(A=@n17, B=@n18.R)))
  n20 = LinearInterpolate(B=ScalarParameter() {defaultValue=0.15, parameterName=BulgeOffset}, Alpha=VertexColor().G)
  n21 = Sine(None=Add(A=Add(A=@n19, B=@n20), B=VertexColor().R))
  Add(
   A=Add(A=Add(A=@n12, B=@n15), B=LinearInterpolate(B=@n16, Alpha=@n21))
   B=MaterialFunctionCall!(
    WSDeformationCenter=VectorParameter() {bUseCustomPrimitiveData=true, group=z_code, parameterName=WSDeformationCenter, primitiveDataIndex=20}.RGB
    Code_DestructionAmount=ScalarParameter() {group=z_code, parameterName=Code_DestructionAmount}
    Code_UseUvForDamage
    Code_UseHitEffects=StaticBoolParameter() {defaultValue=true, group=z_code, parameterName=Code_UseHitEffects}
    Code_Emissive
    WSDamageRadius=ScalarParameter() {bUseCustomPrimitiveData=true, defaultValue=1, group=z_code, parameterName=WSDamageRadius, primitiveDataIndex=26}
    CrackDamageShader
    Diffuse
    ScalarModDamageCracks
    ScalarOffsetforDamageCracks
    ExposeLerpFunctionForInvertingTheHitDistanceCalc
    InvertDamageFalloff
    CrackWorldPosition
    WorldPosition
    DistanceCullFadeAnimationOffset=ScalarParameter() {bUseCustomPrimitiveData=true, group=z_code, parameterName=DistanceCullFadeAnimationOffset, primitiveDataIndex=12}
   ) {materialFunction=/Game/Packages/jl_tests/Fortnite_HitEffects}.Hit And Shrink WPO
  )
  n22 = TextureSample(UVs=Panner(Coordinate=@n8.UVs with Parallax) {speedY=-0.06}) {texture=/Game/VisualThreatManager/StormVisuals/Textures/2ndInfectedMapLinear}
  n23 = Add(A=Multiply(A=ComponentMask(None=@n22.R) {g=true, r=true}) {constB=0.02}, B=@n8.UVs with Parallax)
  n24 = ComponentMask(None=@n23) {r=true}
  n25 = Add(A=OneMinus(None=ComponentMask(None=@n23) {g=true}), B=ScalarParameter() {defaultValue=0.1, parameterName=Offset})
  n26 = VectorParameter() {parameterName=Panner Speed, defaultValue=<rgba -0.01 0 0 0>}
  n27 = Multiply(A=AppendVector(A=@n26.R, B=@n26.R)) {constB=-1}
  n28 = VertexColor()
  n29 = TextureSample(UVs=Panner(Coordinate=AppendVector(A=@n24, B=@n25), Speed=LinearInterpolate(B=@n27, Alpha=@n28.G)) {speedX=-0.01}) {samplerType=SAMPLERTYPE_Normal, texture=/Game/Effects/Fort_Effects/Textures/Tile/MuzzlePattern_TileX}
  n30 = Add(A=@n29.R, B=@n29.R)
  DepthFade(
   Opacity=Multiply(A=Saturate(None=@n30), B=ScalarParameter() {defaultValue=0.75, parameterName=Master Opacity})
   FadeDistance=ScalarParameter() {defaultValue=10, parameterName=Fade Distance}
  )
  n31 = Transform(None=VectorParameter() {parameterName=FakeLightVector, defaultValue=<rgba 0 1 0 0>}.RGB) {transformSourceType=TRANSFORMSOURCE_Local}
  n32 = TextureSample(UVs=Panner(Coordinate=@n23) {speedY=-0.06}) {samplerType=SAMPLERTYPE_Normal, texture=/Game/Packages/Fortnite_UniqueMaterials_1/Textures/T_WaterSoft_N}
  n33 = ScalarParameter() {defaultValue=0.8, parameterName=NormalFlatness}
  n34 = ReflectionVectorWS(CustomWorldNormal=Transform(None=MaterialFunctionCall(Normal=@n32.RGB, Flatness=@n33) {materialFunction=&FlattenNormal}.Result) {transformSourceType=TRANSFORMSOURCE_View})
  n35 = Power(Base=Clamp(None=DotProduct(A=@n31, B=@n34))) {constExponent=20}
  Multiply(
   A=Multiply(A=Clamp(None=@n35) {maxDefault=5}, B=VectorParameter() {parameterName=EmissiveColor, defaultValue=<rgba 1.2 0.2 0 1>}.RGB)
   B=ScalarParameter() {defaultValue=0.5, parameterName=ReflectionIntensity}
  ) {constB=0.1}
 ]
}
